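# Sound cue for the looping music of the Athena ChickenMoves emote, routed to the first-person emote music sound class.
# A single looping wave player serves as the root and plays the ChickenMoves music loop with no other processing nodes.
# The cue references the shared first-person emote music concurrency rule and includes a send to the music audio analysis submix.
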SoundCue /Game/Athena/Sounds/Emotes/ChickenMoves/Athena_Emote_ChickenMoves_Music_Cue
root=WavePlayer_1
nodes [{id=WavePlayer_1, properties {bLooping=true, soundWaveAssetPtr=&/Athena_Emote_ChickenMoves_Music_Loop}}]
properties {
 maxDistance=2097152
 soundClassObject=/Game/Sounds/FortSoundClassesAndModes/Fort_Music_Emote_1P_PSM
 volumeMultiplier=0.95
 concurrencyOverrides {maxCount=3, resolutionRule=StopOldest}
 concurrencySet [/Game/Sounds/Concurrency_Rules/Athena_EmoteMusic_1P_Conc]
 soundSubmixSends [{minSendLevel=0, soundSubmix=&Fortnite_AudioAnalysis_Music}]
}
